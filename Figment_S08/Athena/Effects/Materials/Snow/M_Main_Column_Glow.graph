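# Unlit, two-sided translucent material for a glowing, steamy column effect, filed under Athena snow effects.
# Panning layers of one chromatic smoke texture, weighted along V, form a noise mask that blends two tint colors.
# Color shifts along V toward per-channel smoke samples with small UV offsets for color fringing, then passes through DeriveHDRfromLDR.
# Opacity combines noise, vertical masks and inverted Fresnel for soft edges, adding near-camera fade except on Low quality.
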Material /Figment_S08/Athena/Effects/Materials/Snow/M_Main_Column_Glow
graph {
 properties {blendMode=BLEND_Translucent, pixelDepthOffsetMode=PDOM_Legacy, shadingModel=MSM_Unlit, twoSided=true, usedShadingModels=Unlit}
 nodes [
  n1 = ComponentMask(None=TextureCoordinate()) {g=true}
  n2 = TextureCoordinate()
  n3 = AppendVector(A=ScalarParameter() {defaultValue=2, parameterName=Steam_Scale_U}, B=ScalarParameter() {defaultValue=2, parameterName=Steam_Scale_V})
  n4 = Time()
  n5 = Constant2Vector() {g=0.125}
  n6 = TextureSample(UVs=Add(A=Multiply(A=@n2, B=@n3), B=Multiply(A=@n4, B=@n5))) {samplerSource=SSM_Wrap_WorldGroupSettings, texture=/Game/Effects/Fort_Effects/Textures/Tile/T_Inky_Smoke_Tile_Chromatic}
  n7 = TextureCoordinate()
  n8 = Multiply(A=@n3) {constB=1.5}
  n9 = Multiply(A=Time(), B=Constant2Vector() {g=0.1, r=0.0325})
  n10 = TextureSample(UVs=Add(A=Multiply(A=@n7, B=@n8), B=@n9)) {samplerSource=SSM_Wrap_WorldGroupSettings, texture=&T_Inky_Smoke_Tile_Chromatic}
  n11 = ScalarParameter() {defaultValue=2, parameterName=Steam_Scale_2_U}
  n12 = ScalarParameter() {defaultValue=0.25, parameterName=Steam_Scale_2_V}
  n13 = Add(A=Multiply(A=TextureCoordinate(), B=AppendVector(A=@n11, B=@n12)), B=Multiply(A=Time(), B=Constant2Vector() {g=0.125}))
  n14 = ScalarParameter() {defaultValue=2, parameterName=UV By Life Power}
  n15 = Multiply(
   A=Power(Base=@n1) {constExponent=3}
   B=LinearInterpolate(
    A=Add(A=@n6.G, B=@n10.G)
    B=TextureSample(UVs=@n13) {samplerSource=SSM_Wrap_WorldGroupSettings, texture=&T_Inky_Smoke_Tile_Chromatic}.G
    Alpha=Power(Base=@n1, Exp=@n14)
   )
  )
  n16 = Multiply(A=TextureCoordinate(), B=Multiply(A=@n3) {constB=2.5})
  n17 = Multiply(A=Time(), B=Constant2Vector() {g=0.125})
  n18 = Add(A=Add(A=@n16) {constB=-0.025}, B=@n17)
  n19 = Multiply(A=TextureCoordinate(), B=Multiply(A=@n3) {constB=3})
  n20 = Multiply(A=Time(), B=Constant2Vector() {g=0.1, r=0.0325})
  n21 = TextureSample(UVs=Add(A=Add(A=@n19) {constB=-0.025}, B=@n20)) {samplerSource=SSM_Wrap_WorldGroupSettings, texture=&T_Inky_Smoke_Tile_Chromatic}
  n22 = Multiply(A=Add(
    A=TextureSample(UVs=@n18) {samplerSource=SSM_Wrap_WorldGroupSettings, texture=&T_Inky_Smoke_Tile_Chromatic}.R
    B=@n21.R
   )) {constB=0.5}
  n23 = TextureSample(UVs=Add(A=@n16, B=@n17)) {samplerSource=SSM_Wrap_WorldGroupSettings, texture=&T_Inky_Smoke_Tile_Chromatic}
  n24 = Multiply(A=Add(A=@n23.G, B=@n21.G)) {constB=0.5}
  n25 = Add(A=Add(A=@n19) {constB=-0.0425}, B=@n20)
  n26 = Multiply(A=Add(
    A=@n23.B
    B=TextureSample(UVs=@n25) {samplerSource=SSM_Wrap_WorldGroupSettings, texture=&T_Inky_Smoke_Tile_Chromatic}.B
   )) {constB=0.5}
  n27 = LinearInterpolate(
   A=LinearInterpolate(
    A=VectorParameter() {parameterName=Secondary Color, defaultValue=<rgba 0.9389 0.6342 1 1>}.RGB
    B=VectorParameter() {parameterName=Prrimary Color, defaultValue=<rgba 2.2143 1.4263 2.5 1>}.RGB
    Alpha=Power(Base=@n15) {constExponent=2.25}
   )
   B=Saturate(None=MaterialFunctionCall(X=@n22, Y=@n24, Z=@n26) {materialFunction=&MakeFloat3}.Result)
   Alpha=Saturate(None=Power(Base=@n1))
  )
  MaterialFunctionCall(LDR Input=@n27, HDR Tint=@n27) {materialFunction=&DeriveHDRfromLDR}
  n28 = Power(Base=ComponentMask(None=TextureCoordinate()) {g=true}, Exp=Constant() {r=4})
  n29 = TextureSample(UVs=Add(A=TextureCoordinate(), B=@n9)) {samplerSource=SSM_Wrap_WorldGroupSettings, texture=&T_Inky_Smoke_Tile_Chromatic}
  n30 = Add(A=@n15, B=Multiply(A=Add(A=@n28, B=@n29.G), B=@n28))
  n31 = Saturate(None=Multiply(A=OneMinus(None=@n1)) {constB=2})
  n32 = ScalarParameter() {defaultValue=0.2, parameterName=Fresnel Exponent}
  n33 = Multiply(A=OneMinus(None=Fresnel(ExponentIn=@n32))) {constB=3}
  n34 = Subtract(A=ComponentMask(None=TextureCoordinate()) {g=true}, B=ScalarParameter() {defaultValue=0.5, parameterName=Cylinder Mask Offset})
  n35 = Multiply(A=Multiply(A=Multiply(A=@n30, B=@n31), B=Power(Base=@n33) {constExponent=2.25}), B=Saturate(None=Multiply(A=@n34)))
  n36 = Multiply(A=@n35, B=MaterialFunctionCall() {materialFunction=&Near_Camera_Fade}.Result)
  QualitySwitch(Default=@n36, Low=@n35, High=@n36, Medium=@n36)
 ]
}
