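# Translucent unlit effect material for ground-level snow smoke in Figment Season 8.
# Layered smoke texture samples at different tilings and speed-scaled UV offsets blend along a powered vertical UV gradient.
# A procedural UV1 edge mask, combined with pattern and gradient textures, shapes that smoke before alpha scaling.
# A separate branch brightens a time-of-day shadow color; the Time-times-speed product feeds none of the shown UV offsets.
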Material /Figment_S08/Athena/Effects/Materials/Snow/M_Snow_Ground_Smoke
graph {
 properties {
  blendMode=BLEND_Translucent
  pixelDepthOffsetMode=PDOM_Legacy
  shadingModel=MSM_Unlit
  translucencyPass=MTP_BeforeDOF
  usedShadingModels=Unlit
 }
 nodes [
  n1 = ScalarParameter() {defaultValue=1, parameterName=speed}
  n2 = Add(A=Add(A=TextureCoordinate() {vTiling=0.6}) {constB=0.325}, B=Multiply(A=Constant2Vector() {g=0.2}, B=@n1))
  n3 = VectorParameter() {defaultValue=<rgba 1 1 0 1>}
  n4 = MaterialFunctionCall(R=@n3.R, G=@n3.R) {materialFunction=&AppendMany}
  n5 = Add(A=TextureCoordinate() {uTiling=0.5, vTiling=0.45}, B=Multiply(A=Constant2Vector() {g=0.325, r=0.125}, B=@n1))
  n6 = Add(
   A=TextureSample(UVs=Multiply(A=@n2, B=@n4.RG)) {samplerSource=SSM_Wrap_WorldGroupSettings, texture=/Figment_S08/Athena/Effects/Textures/Snow/T_Thick_Smoke_05}.G
   B=TextureSample(UVs=Multiply(A=@n5, B=@n4.RG)) {samplerSource=SSM_Wrap_WorldGroupSettings, texture=&T_Thick_Smoke_05}.G
  )
  n7 = TextureSample(UVs=@n5) {
   samplerSource=SSM_Wrap_WorldGroupSettings
   texture=/Figment_S08/Packages/Fortress_SharedMaterials/Textures/T_Blowing_Smoke_Pattern2_Tile_DISP
  }
  Add(
   A=@n6
   B=Multiply(
    A=TextureSample(UVs=@n2) {samplerSource=SSM_Wrap_WorldGroupSettings, texture=/Game/Effects/Fort_Effects/Textures/Tile/T_Inky_Smoke_Tile_Chromatic}.G
    B=@n7.G
   )
  )
  n8 = TextureCoordinate() {vTiling=2}
  n9 = Multiply(A=Constant2Vector() {g=0.25, r=0.01}, B=@n1)
  n10 = TextureSample(UVs=Multiply(A=Add(A=@n8, B=@n9), B=@n4.RG)) {automaticViewMipBias=false, samplerSource=SSM_Wrap_WorldGroupSettings, texture=&T_Thick_Smoke_05}
  n11 = Constant2Vector() {g=0.25, r=-0.0125}
  n12 = Multiply(A=Add(A=TextureCoordinate() {uTiling=0.5, vTiling=1.25}, B=Multiply(A=@n11, B=@n11)), B=@n4.RG)
  n13 = ComponentMask(None=TextureCoordinate()) {g=true}
  n14 = LinearInterpolate(
   A=@n6
   B=Add(
    A=@n10.G
    B=TextureSample(UVs=@n12) {samplerSource=SSM_Wrap_WorldGroupSettings, texture=&T_Thick_Smoke_05}.G
   )
   Alpha=Power(Base=OneMinus(None=@n13), Exp=ScalarParameter() {defaultValue=3, parameterName=Slow Down UV Power})
  )
  n15 = TextureCoordinate() {coordinateIndex=1}
  n16 = ComponentMask(None=@n15) {r=true}
  n17 = ComponentMask(None=@n15) {g=true}
  n18 = OneMinus(None=@n17)
  n19 = Multiply(A=@n17, B=Multiply(A=@n18, B=@n18))
  n20 = Multiply(A=Multiply(A=@n16, B=OneMinus(None=@n16)), B=@n19)
  n21 = ScalarParameter() {defaultValue=32, parameterName=Mask Intensity}
  n22 = TextureCoordinate()
  n23 = TextureSample(UVs=@n22) {texture=&/T_FX_Vault_Smoke_Gradient}
  n24 = Multiply(A=Saturate(None=Multiply(A=@n20, B=@n21) {constB=48}), B=Multiply(A=Add(A=@n7.G, B=@n23.RGB), B=@n23.RGB))
  Multiply(
   A=Multiply(A=Multiply(A=@n14, B=@n24), B=ScalarParameter() {defaultValue=1, parameterName=Final_Alpha})
   B=ScalarParameter() {defaultValue=1.5}
  )
  n25 = Power(Base=@n17) {constExponent=6}
  Multiply(A=@n19, B=Multiply(A=Multiply(A=@n25, B=@n18)) {constB=10})
  Multiply(A=Time(), B=@n1)
  Multiply(
   A=Constant3Vector() {constant=<rgba 5 5 5 0>}
   B=MaterialFunctionCall(
    Color_Morning=VectorParameter() {group=Colors, parameterName=ShadowMorning, defaultValue=<rgba 0.5 0.3633 0.349 0>}.RGB
    Color_Day=VectorParameter() {group=Colors, parameterName=ShadowDay, defaultValue=<rgba 0.4671 0.4531 0.6 0>}.RGB
    Color_Evening=VectorParameter() {group=Colors, parameterName=ShadowEvening, defaultValue=<rgba 0.3703 0.3458 0.4 1>}.RGB
    Color_Night=VectorParameter() {group=Colors, parameterName=ShadowNight, defaultValue=<rgba 0.25 0.2445 0.2279 0>}.RGB
   ) {materialFunction=&Color_Time_of_Day}.Result
  )
  TextureSample(UVs=Add(A=@n22, B=ScalarParameter() {parameterName=Mask Offset}) {constB=0.2}) {texture=/Game/Effects/Fort_Effects/Textures/Grad/gradient_stasis}
 ]
}
