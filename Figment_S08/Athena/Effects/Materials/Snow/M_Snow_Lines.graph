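# Translucent, two-sided unlit mesh-particle material for glowing snow line effects, blue by default.
# A gradient texture blends two colors and masks opacity, with time-of-day and HDR adjustment on the emissive.
# Dynamic particle parameters drive intensity, opacity and a UV1 power-on wipe, alongside view-angle fade and panning smoke-texture breakup.
# An optional UseImpactAnim switch adds a hit-location ripple, health-driven flickering glow and a lake-structure hit wobble function.
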Material /Figment_S08/Athena/Effects/Materials/Snow/M_Snow_Lines
graph {
 properties {
  bUsedWithMeshParticles=true
  blendMode=BLEND_Translucent
  pixelDepthOffsetMode=PDOM_Legacy
  shadingModel=MSM_Unlit
  twoSided=true
  usedShadingModels=Unlit
 }
 nodes [
  n1 = ParticleColor()
  n2 = Time()
  n3 = Clamp(None=Divide(A=@n2, B=ScalarParameter() {defaultValue=0.5, parameterName=ImpactDuration}))
  n4 = SphereMask(
   A=WorldPosition().XYZ
   B=VectorParameter() {parameterName=HitWave-Loc, defaultValue=<rgba 50 0 -50 0>}.RGB
   Radius=Multiply(A=Power(Base=@n3) {constExponent=0.55}, B=ScalarParameter() {defaultValue=300, parameterName=RippleRadius})
  ) {hardnessPercent=0}
  n5 = Multiply(A=Clamp(None=Sine(None=@n4)), B=@n4)
  n6 = Power(Base=ScalarParameter() {defaultValue=1, group=z_code, parameterName=Health}) {constExponent=0.6}
  n7 = LinearInterpolate(Alpha=@n6) {constA=4, constB=0.1}
  n8 = Add(B=Multiply(A=@n7) {constB=3}) {constA=5}
  n9 = OneMinus(None=@n3)
  n10 = Subtract(A=@n2, B=ScalarParameter() {parameterName=GameTime})
  n11 = LinearInterpolate(
   A=ScalarParameter() {defaultValue=3, parameterName=GlowDurationMax}
   B=ScalarParameter() {defaultValue=0.5, parameterName=GlowDurationMin}
   Alpha=@n6
  )
  n12 = OneMinus(None=Clamp(None=Divide(A=@n10, B=@n11)))
  n13 = Multiply(A=Time(), B=LinearInterpolate(Alpha=@n6) {constA=16})
  n14 = Add(A=ConstantBiasScale(None=Sine(None=@n13))) {constB=0.1}
  n15 = Add(A=Multiply(A=Multiply(A=@n5, B=@n8) {constB=4}, B=Power(Base=@n9)), B=Multiply(A=Power(Base=@n12), B=Multiply(A=@n14, B=@n7))) {constB=0}
  n16 = Multiply(A=@n15)
  n17 = Power(
   Base=TextureSample() {samplerType=SAMPLERTYPE_LinearGrayscale, texture=/Figment_S08/Athena/Effects/Textures/Snow/t_snow_mesh_gradients}.G
   Exp=ScalarParameter() {defaultValue=2, parameterName=Emissive Mask Power}
  )
  n18 = LinearInterpolate(
   A=VectorParameter() {parameterName=Color, defaultValue=<rgba 0 0.0182 0.349 1>}.RGB
   B=VectorParameter() {parameterName=Color 2, defaultValue=<rgba 0 0.5 1 1>}.RGB
   Alpha=@n17
  )
  n19 = StaticSwitchParameter(True=Multiply(A=Add(A=@n16) {constB=2}, B=@n18), False=@n18) {parameterName=UseImpactAnim}
  n20 = ScalarParameter() {defaultValue=0.65, parameterName=Night Multiplier}
  n21 = Multiply(
   A=@n1.RGB
   B=MaterialFunctionCall(Color_Morning=@n19, Color_Day=@n19, Color_Evening=@n19, Color_Night=Multiply(A=@n19, B=@n20) {constB=0.45}) {materialFunction=&Color_Time_of_Day}.Result
  )
  n22 = Clamp(None=@n21)
  n23 = StaticSwitchParameter(
   True=DynamicParameter() {defaultValue=<rgba 1 1 0 1>, paramNames [Param1, HDRIntensity, Power_On_Lerp, Final_Opacity]}.HDRIntensity
   False=ScalarParameter() {defaultValue=5, parameterName=HDRIntensity}
  ) {defaultValue=true, parameterName=Use Dynamic Param?}
  n24 = MaterialFunctionCall(
   LDR Input=@n22
   HDR Tint=@n21
   HDR Intensity Multiplier=MaterialFunctionCall(Value_Morning=@n23, Value_Day=@n23, Value_Evening=@n23, Value_Night=Multiply(A=@n23, B=@n20) {constB=0.5}) {materialFunction=&Scalar_Time_of_Day}.Result
  ) {materialFunction=&DeriveHDRfromLDR}
  FeatureLevelSwitch(Default=@n24.RemappedRange, ES2=@n22, ES3_1=@n22, SM4_REMOVED=@n24.RemappedRange)
  n25 = StaticSwitchParameter(
   True=DynamicParameter() {defaultValue=<rgba 1 1 0 1>, paramNames [Param1, HDRIntensity, Power_On_Lerp, Final_Opacity]}.Final_Opacity
   False=ScalarParameter() {parameterName=Final_Opacity}
  ) {defaultValue=true, parameterName=Use Dynamic Param?}
  n26 = OneMinus(None=ComponentMask(None=TextureCoordinate() {coordinateIndex=1}) {g=true})
  n27 = LinearInterpolate(
   A=ScalarParameter() {defaultValue=-1, parameterName=Power On Low}
   B=ScalarParameter() {defaultValue=3.5, parameterName=Power On High}
   Alpha=StaticSwitchParameter(
    True=DynamicParameter() {defaultValue=<rgba 1 1 0 1>, paramNames [Param1, HDRIntensity, Power_On_Lerp, Final_Opacity]}.Power_On_Lerp
    False=ScalarParameter() {parameterName=Power On Lerp}
   ) {defaultValue=true, parameterName=Use Dynamic Param?}
  ) {constA=-1, constB=3.5}
  n28 = Saturate(None=Ceil(None=Add(A=@n26, B=@n27)))
  n29 = Constant()
  n30 = Multiply(A=@n17, B=Add(A=Multiply(A=@n25, B=@n28), B=StaticSwitchParameter(True=@n15, False=@n29) {parameterName=UseImpactAnim}))
  n31 = ScalarParameter() {parameterName=Angle Fade Sensitivity}
  Multiply(
   A=@n1.A
   B=Multiply(
    A=@n30
    B=MaterialFunctionCall("FadeSensitivity (optional)"=@n31) {materialFunction=&FadeBasedOnViewAngle}.Result
   )
  )
  n32 = VectorParameter() {parameterName=Secondary Energy UV, defaultValue=<rgba 1 1 0 0>}
  n33 = Add(
   A=Multiply(A=TextureCoordinate() {coordinateIndex=1}, B=ComponentMask(None=@n32.RGB) {g=true, r=true})
   B=Multiply(A=Time(), B=AppendVector(A=@n32.B, B=@n32.B))
  )
  Multiply(
   A=@n30
   B=LinearInterpolate(
    A=ScalarParameter() {parameterName=Energy Texture Multiplier Lerp}
    B=@n28
    Alpha=TextureSample(UVs=@n33) {texture=/Game/Effects/Fort_Effects/Textures/Tile/T_Inky_Smoke_Tile_Chromatic}.G
   )
  )
  ScalarParameter() {defaultValue=1, parameterName=Active}
  StaticSwitchParameter(
   True=MaterialFunctionCall!(VibrationAmount, WobbleScale, WobbleDuration) {materialFunction=/Figment_S08/Environments/Sets/DigSite/Materials/MF_LakeStructure_HitWPO}.Result
   False=Constant()
  ) {parameterName=UseImpactAnim}
  Time() {bOverride_Period=true, period=0.5}
  Time() {bOverride_Period=true, period=2}
  VectorParameter() {parameterName=ColorWarmup, defaultValue=<rgba 0.35 0.0985 0.0139 0>}
  VectorParameter() {parameterName=ColorWarmup2, defaultValue=<rgba 1 0.6403 0.0978 0>}
 ]
}
